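# Widget that updates Slate post buffer 0, with or without blur, judging by its config key for front-end widgets.
# Its canvas holds two full-stretch PostBufferUpdate widgets, one with a blur updater and one without, slightly overscanned through negative offsets.
# On initialization it reads EnableBlurForFrontEndWidgets, treating a failed read as true, and removes the variant that does not apply.
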WidgetBlueprint /FortUILibrary/Blueprints/PostProcessor/WBP_CaptureForPostBufferUpdate
graphs [EventGraph {nodes [
   Event OnInitialized()
   n1 = GetConfigDataBool!(EntryName:String=EnableBlurForFrontEndWidgets) -> (bValueOut:bool, ReturnValue:bool)
   > n2 = Branch(Condition=Select(Option 0:bool=true, Option 1:bool=@n1, Index=@n1.ReturnValue) -> (ReturnValue:bool))
   > RemoveFromParent(self=Get PostBufferUpdate_BlurDisabled() -> (PostBufferUpdate_BlurDisabled:Post Buffer Update Object Reference))
   RemoveFromParent(execute=@n2.else, self=Get PostBufferUpdate() -> (PostBufferUpdate:Post Buffer Update Object Reference))
  ]}]
variables []
widgets {
 info {inheritedWidgetCount=0, parentClass=&UserWidget, rootWidgetClass=&CanvasPanel, widgetCount=3}
 widgets [
  HackForNegativeOffset {widgetClassPath=&CanvasPanel, widget=HackForNegativeOffset}
  PostBufferUpdate {
   widgetClassPath=/Script/UMG.PostBufferUpdate
   bIsVariable=true
   parent=HackForNegativeOffset
   widget=PostBufferUpdate
   properties {
    visibility=HitTestInvisible
    updateBufferInfos [{
      bufferToUpdate=ESlatePostRT_0
      postParamUpdater=&self:WidgetTree.PostBufferUpdate.PostBufferBlurUpdater_0
     }]
   }
   slot {class=CanvasPanelSlot, properties {layoutData {alignment=<0 0>, anchors {maximum=<1 1>, minimum=<0 0>}, offsets=<ltrb -2 -2 -2 -2>}}}
  }
  PostBufferUpdate_BlurDisabled {
   widgetClassPath=&PostBufferUpdate
   bIsVariable=true
   parent=HackForNegativeOffset
   widget=PostBufferUpdate_BlurDisabled
   properties {visibility=HitTestInvisible, updateBufferInfos [{bufferToUpdate=ESlatePostRT_0, postParamUpdater=null}]}
   slot {class=CanvasPanelSlot, properties {layoutData {alignment=<0 0>, anchors {maximum=<1 1>, minimum=<0 0>}, offsets=<ltrb -2 -2 -2 -2>}}}
  }
 ]
}
